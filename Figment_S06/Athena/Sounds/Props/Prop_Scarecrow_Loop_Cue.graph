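# Sound cue that plays a looping sound for a scarecrow prop, routed through the third-person dance music sound class.
# A looping wave player feeds a modulator at the root, with the overall volume lowered.
# Its own attenuation override gives a short falloff radius and uses interpolated listener focus to quieten the sound when out of focus.
# A concurrency override allows one instance at a time, stopping the quietest.
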SoundCue /Figment_S06/Athena/Sounds/Props/Prop_Scarecrow_Loop_Cue
root=Modulator_0
nodes [
 {id=Modulator_0, children [WavePlayer_2]}
 {id=WavePlayer_2, properties {bLooping=true, soundWaveAssetPtr=&/Prop_Scarecrow_Loop}}
]
properties {
 bOverrideAttenuation=true
 bOverrideConcurrency=true
 maxDistance=512
 soundClassObject=/Game/Sounds/FortSoundClassesAndModes/Fort_DanceMusic_3P
 volumeMultiplier=0.65
 attenuationOverrides {
  bEnableFocusInterpolation=true
  bEnableListenerFocus=true
  falloffDistance=512
  focusAttackInterpSpeed=0.5
  focusReleaseInterpSpeed=0.5
  nonFocusVolumeAttenuation=0.2
  attenuationShapeExtents {x=0}
 }
 concurrencyOverrides {maxCount=1, resolutionRule=StopQuietest}
}
